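# Default Lit, Nanite-enabled material for NeonCity neon signs that shows a texture as an emissive LCD-style screen.
# UVs are snapped to a tiled grid with Ceil to pixelate the sign texture, which is multiplied by an LCD panel mask.
# An optional static switch adds a panning gradient brightness boost masked by rounded vertex red.
# A Fresnel falloff scales the glow before MF_NeonCityEmissiveBalance, alongside a glass-grime lerp and flat normal.
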Material /Game/Environments/Asteria/Sets/NeonCity/Techart/NeonSigns/Shapes/M_SimpleTexturedNeon
graph {
 properties {bUsedWithNanite=true, pixelDepthOffsetMode=PDOM_Legacy, usedShadingModels=Default Lit}
 nodes [
  Constant()
  Constant() {r=0.6}
  LinearInterpolate(Alpha=TextureSample() {samplerType=SAMPLERTYPE_LinearColor, texture=/Game/Environments/Apollo/Sets/Tinfoil/Textures/T_Tinfoil_GlassGrime_M}.B) {constA=0.1, constB=0.4}
  n1 = LinearInterpolate(Alpha=Fresnel(ExponentIn=ScalarParameter() {defaultValue=0.5, parameterName=Emissive Fresnel Angle}, BaseReflectFractionIn=Constant())) {constA=1, constB=0.1}
  n2 = ScalarParameter() {defaultValue=375, parameterName=LCDTiling}
  n3 = Multiply(A=TextureCoordinate(), B=@n2)
  n4 = TextureSampleParameter2D(UVs=Divide(A=Ceil(None=@n3), B=@n2)) {parameterName=Texture, texture=&/SushiFish}
  n5 = ScalarParameter() {defaultValue=5, parameterName=PanningEmissiveBoost}
  n6 = TextureCoordinate()
  n7 = ScalarParameter() {defaultValue=1, parameterName=GradSize}
  n8 = ScalarParameter() {defaultValue=-0.5, parameterName=Scroll X}
  n9 = ScalarParameter() {parameterName=Scroll Y}
  n10 = TextureSample(UVs=Panner(Coordinate=Multiply(A=@n6, B=@n7), Time=Time(), Speed=AppendVector(A=@n8, B=@n9))) {texture=/Game/Gadgets/Assets/Misty_Bop/Effects/T_MediumGradient_Horizontal}
  n11 = Round(None=VertexColor().R)
  n12 = StaticSwitchParameter(True=LinearInterpolate(A=@n4.RGB, B=Multiply(A=@n4.RGB, B=@n5), Alpha=Multiply(A=@n10.R, B=@n11)), False=@n4.RGB) {parameterName=PanneronRedChannel}
  n13 = TextureSample(UVs=@n3) {samplerType=SAMPLERTYPE_Masks, texture=/Game/Environments/Asteria/Sets/NeonCity/Techart/Textures/T_LCD_Panels_LCDSoft}
  MaterialFunctionCall!(Emissive=Multiply(A=Saturate(None=@n1), B=Multiply(A=@n12, B=@n13.RGB))) {materialFunction=/Game/Environments/Asteria/Sets/NeonCity/Techart/Screen/MaterialFunctions/MF_NeonCityEmissiveBalance}
  TextureSampleParameter2D() {parameterName=FlatNormal, samplerType=SAMPLERTYPE_Normal, texture=/Engine/EngineMaterials/FlatNormal}
 ]
}
